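; BP_PinkOatmeal_VisualBubble is a visual bubble actor for the PinkOatmeal environmental item.
; Its construction script uniformly scales the Bubble component from the BoundsMeshRadius variable times a small constant, so the bubble size follows that radius.
; BeginPlay, overlap and tick events are empty stubs, so it runs no gameplay logic of its own.
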(event EventBeginPlay)

(event Collision|EventActorBeginOverlap
  (params (OtherActor :type Actor%20Object%20Reference)))

(event EventTick
  (params (DeltaSeconds :type Float%20%28single-precision%29)))

(fn ConstructionScript
  (Transformation|SetWorldScale3D (Variables|Default|GetBubble) (Math|Vector|MakeVector (* (Variables|Default|GetBoundsMeshRadius) 0.0501) (* (Variables|Default|GetBoundsMeshRadius) 0.0501) (* (Variables|Default|GetBoundsMeshRadius) 0.0501))))
