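; PinkOatmeal SleepyTime actor that periodically spawns "star" drop actors (BGA_PinkOatmeal_SleepyTime_Drop) from points on its bubble radius.
; On the server, once playlist data is ready, a scalable-float toggle gates a looping spawn timer with a tunable interval.
; Each spawn point uses a random tilt from actor up and random yaw, with the drop oriented toward a random target around the actor's center.
; FindStarSpawnLoc still draws a debug line.
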(event EventBeginPlay
  ;[1 1 1] Spawn the stars on a delay
  (CallFunction|HandleStarSpawn)
  (|CallParentFunction))

(event Custom|SpawnStar
  ;[1 1 1] Spawn the stars on a delay
  (bind _returnvalue (CallFunction|FindStarSpawnLoc))
  (bind _returnvalue_2 (CallFunction|FindEndTargetAngle _returnvalue))
  (Game|SpawnActorfromClass "/Figment_S10/Athena/Items/EnvironmentalItems/PinkOatmeal_10_4/SleepyTime/BGA_PinkOatmeal_SleepyTime_Drop.BGA_PinkOatmeal_SleepyTime_Drop_C" (Math|Transform|MakeTransform _returnvalue (Math|Conversions|RotationFromXVector _returnvalue_2)) "AlwaysSpawn" "OverrideRootScale" self 0 0))

(event Custom|HandleStarSpawn
  ;[1 1 1] Spawn the stars on a delay
  (if (Networking|IsServer)
    (Class|FortAsyncActionCurrentPlaylistReady|PlaylistDataReadyAsync
      (:OnReady
        (if (Math|Conversions|ToBoolean(Integer) (Math|Float|Floor (ScalableFloat|GetValueatLevel (Variables|Default|GetRow_StarSpawnEnabled))))
          (bind _returnvalue (Utilities|Time|SetTimerbyEvent (delegate Custom|SpawnStar) (ScalableFloat|GetValueatLevel (Variables|Default|GetRow_StarSpawnInterval)) true))
          (Variables|Default|SetTimer_StarSpawnInterval _returnvalue))))))

(event Collision|EventActorBeginOverlap
  (params (OtherActor :type Actor%20Object%20Reference)))

(event EventTick
  (params (DeltaSeconds :type Float%20%28single-precision%29)))

(fn FindEndTargetAngle
  (params (InVect :type Vector))
  (return (- (+ (Transformation|GetActorLocation) (* (Math|Vector|Normalize (Math|Vector|RotateVectorAroundAxis (Transformation|GetActorForwardVector) (Math|Random|RandomFloatinRange 0.0 3.6e+02) "0, 0,1.000000")) (* (CallFunction|GetBubbleRadius) (Math|Random|RandomFloatinRange 0.01 (ScalableFloat|GetValueatLevel (Variables|Default|GetRow_StarEndPointRadiusPercentage)))))) InVect)))

(fn FindStarSpawnLoc
  (bind _radius (CallFunction|GetBubbleRadius))
  (bind _returnvalue (Math|Random|RandomFloatinRange 0.0 3.6e+02))
  (bind _row_starspawnpointanglemax (Variables|Default|GetRow_StarSpawnPointAngleMax))
  (bind _returnvalue_2 (ScalableFloat|GetValueatLevel _row_starspawnpointanglemax))
  (bind _row_starspawnpointanglemin (Variables|Default|GetRow_StarSpawnPointAngleMin))
  (bind _returnvalue_3 (ScalableFloat|GetValueatLevel _row_starspawnpointanglemin))
  (bind _returnvalue_4 (Math|Random|RandomFloatinRange _returnvalue_3 _returnvalue_2))
  (bind _returnvalue_5 (Transformation|GetActorUpVector))
  (bind _returnvalue_6 (Math|Vector|RotateVector _returnvalue_5 (Math|Rotator|MakeRotator _returnvalue_4)))
  (bind _returnvalue_7 (Math|Vector|RotateVector _returnvalue_6 (Math|Rotator|MakeRotator 0.0 0.0 _returnvalue)))
  (bind _returnvalue_8 (Math|Vector|Normalize _returnvalue_7))
  (bind _returnvalue_9 (* _returnvalue_8 _radius))
  (bind _returnvalue_10 (Transformation|GetActorLocation))
  (bind _returnvalue_11 (+ _returnvalue_10 _returnvalue_9))
  (Rendering|Debug|DrawDebugLine _returnvalue_10 _returnvalue_11 "(R=1.000000,G=0.000000,B=0.014146,A=1.000000)" 1.0 5.0)
  (return _returnvalue_11))

(fn ConstructionScript
  (|CallParentFunction))
